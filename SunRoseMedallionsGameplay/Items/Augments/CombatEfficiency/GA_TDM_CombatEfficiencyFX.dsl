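; Passive gameplay ability for the Combat Efficiency augment that plays a firing gameplay cue as the owning pawn's weapon consumes ammo.
; On activation it waits on the firing tag, binding to the current weapon's OnConsumeAmmo when added and unbinding when removed.
; Each ammo event executes the cue on the owner if the weapon's tags match the valid list.
; Ending the ability unbinds and ends both tag waits.
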(event Ability|EventFailedToActivatePassiveAbility
  (params (ActorInfo :type Gameplay%20Ability%20Actor%20Info%20Structure))
  (Development|PrintString "CombatEfficiencyVFX: Activation FAILED" true true "(R=1.000000,G=0.000000,B=0.332315,A=1.000000)" 2e+01))

(event Ability|EventActivateAbility
  (bind _returnvalue (CallFunction|GetAvatarasFortPlayerPawn))
  (bind _firingtag (Variables|Default|GetFiringTag))
  ;[1 1 1] EVENT -> Activate Ability
  (|CallParentFunction)
  (Utilities|IsValid _returnvalue
    (:Is%20Valid
      (bind _output_get (Variables|Default|SetOwningPawn _returnvalue))
      (bind _asynctaskproxy (Ability|Tasks|WaitGameplayTagAddToActor _output_get _firingtag)
        (:then
          (Variables|Default|SetAsyncWaitAttackingStart _asynctaskproxy)
          (bind _asynctaskproxy_2 (Ability|Tasks|WaitGameplayTagRemoveFromActor _output_get _firingtag)
            (:then
              (Variables|Default|SetAsyncWaitAttackingEnd _asynctaskproxy_2))
            (:Removed
              (bind _owningpawn (Variables|Default|GetOwningPawn)
                (:then
                  (Utilities|IsValid (Variables|/Script/FortniteGame.FortPawn|GetCurrentWeapon _owningpawn)
                    (:Is%20Valid
                      (EventDispatchers|UnbindEventfromOnConsumeAmmo (Variables|/Script/FortniteGame.FortPawn|GetCurrentWeapon _owningpawn) (EventDispatchers|CreateEvent :!event Custom|AmmoConsumed)))))
                (:else)))))
        (:Added
          (bind _owningpawn_2 (Variables|Default|GetOwningPawn)
            (:then
              (Utilities|IsValid (Variables|/Script/FortniteGame.FortPawn|GetCurrentWeapon _owningpawn_2)
                (:Is%20Valid
                  (EventDispatchers|BindEventtoOnConsumeAmmo (Variables|/Script/FortniteGame.FortPawn|GetCurrentWeapon _owningpawn_2) (EventDispatchers|CreateEvent :!event Custom|AmmoConsumed)))))
            (:else)))))))

(event Custom|AmmoConsumed
  (params (Quantity :type Integer))
  (CallFunction|ExecuteGameplayCueonOwningPawn))

(event Ability|EventOnEndAbility
  (params (bWasCancelled :type Boolean))
  (Utilities|FlowControl|Sequence
    (:then_0
      (bind _owningpawn (Variables|Default|GetOwningPawn)
        (:then
          (Utilities|IsValid (Variables|/Script/FortniteGame.FortPawn|GetCurrentWeapon _owningpawn)
            (:Is%20Valid
              (EventDispatchers|UnbindEventfromOnConsumeAmmo (Variables|/Script/FortniteGame.FortPawn|GetCurrentWeapon _owningpawn) (EventDispatchers|CreateEvent :!event Custom|AmmoConsumed)))))
        (:else)))
    (:then_1
      (Ability|Async|EndAction (Variables|Default|GetAsyncWaitAttackingStart))
      (Ability|Async|EndAction (Variables|Default|GetAsyncWaitAttackingEnd)))
    (:then_2
      (|CallParentFunction bWasCancelled))))

(fn ExecuteGameplayCueOnOwningPawn
  (bind _owningpawn (Variables|Default|GetOwningPawn)
    (:then
      (if (Tags|WeaponDataHasAnyMatchingGameplayTags (Variables|/Script/FortniteGame.FortPawn|GetCurrentWeapon _owningpawn) (Variables|Default|GetValidWeaponTags))
        (Ability|ExecuteGameplayCueOnOwner (Variables|Default|GetGCTag_AresFiringFX) (Class|AthenaFunctionLibrary|MakeSimpleFortEffectContext _owningpawn (Variables|/Script/FortniteGame.FortPawn|GetCurrentWeapon _owningpawn)))))
    (:else)))
